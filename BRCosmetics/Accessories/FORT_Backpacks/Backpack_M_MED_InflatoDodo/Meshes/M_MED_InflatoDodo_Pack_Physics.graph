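# Physics asset for the InflatoDodo backpack cosmetic, defining collision bodies and joints for its jointed dynamic section.
# Six wide capsules on the backpack pelvis and spine bones approximate the torso, linked by fully free constraints.
# A five-capsule dyn_pack chain, narrowing toward the tip, extends from a small pack sphere.
# Its joints have tight twist and modest swing limits, and no constraint links it to the spine bodies.
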PhysicsAsset /BRCosmetics/Accessories/FORT_Backpacks/Backpack_M_MED_InflatoDodo/Meshes/M_MED_InflatoDodo_Pack_Physics
bodies {
 backpack_pelvis [{shapeType=Capsule, length=9.5, radius=13.5, shapeName=backpack_pelvis_capsule, center=<1.2688 -0.3056 0>, rotation=<rot 0 -90 0>}]
 backpack_spine_01 [{shapeType=Capsule, length=8, radius=14, shapeName=backpack_spine_01_capsule, center=<2.5829 0.57 0>, rotation=<rot 0 -90 0>}]
 backpack_spine_02 [{shapeType=Capsule, length=6, radius=14, shapeName=backpack_spine_02_capsule, center=<-0.0307 0.4638 0.0001>, rotation=<rot 0 -90 0>}]
 backpack_spine_03 [{shapeType=Capsule, length=8, radius=14, shapeName=backpack_spine_03_capsule, center=<0.6911 0.0941 0.0001>, rotation=<rot 0 -90 0>}]
 backpack_spine_04 [{shapeType=Capsule, length=14, radius=11, shapeName=backpack_spine_04_capsule, center=<2.766 0.7765 0.0002>, rotation=<rot 0 -90 0>}]
 backpack_spine_05 [{shapeType=Capsule, length=10.5, radius=12.5, shapeName=backpack_spine_05_capsule, center=<-2.2795 1.8109 0>, rotation=<rot 0 -90 0>}]
 dyn_pack_01 [{shapeType=Capsule, length=9.8842, radius=4.5762, shapeName=dyn_pack_01_capsule, center=<6.9643 0 0>, rotation=<rot 90 0 0>}]
 dyn_pack_02 [{shapeType=Capsule, length=6.2661, radius=3.8644, shapeName=dyn_pack_02_capsule, center=<5.6799 0 0>, rotation=<rot 90 0 0>}]
 dyn_pack_03 [{shapeType=Capsule, length=4.9613, radius=3.3062, shapeName=dyn_pack_03_capsule, center=<5.8018 0 0>, rotation=<rot 90 0 0>}]
 dyn_pack_04 [{shapeType=Capsule, length=5.3813, radius=3.0309, shapeName=dyn_pack_04_capsule, center=<5.3213 0 0>, rotation=<rot 90 0 0>}]
 dyn_pack_05 [{shapeType=Capsule, length=12.8803, radius=2.608, shapeName=dyn_pack_05_capsule, center=<8.7359 0 0>, rotation=<rot 90 0 0>}]
 pack [{shapeType=Sphere, radius=2, shapeName=pack_sphere}]
}
constraints [
 {
  bone1Name=backpack_spine_01
  bone2Name=backpack_pelvis
  swing1LimitDegrees=45
  swing1Motion=Free
  swing2LimitDegrees=45
  swing2Motion=Free
  twistLimitDegrees=45
  twistMotion=Free
 }
 {
  bone1Name=backpack_spine_02
  bone2Name=backpack_spine_01
  swing1LimitDegrees=45
  swing1Motion=Free
  swing2LimitDegrees=45
  swing2Motion=Free
  twistLimitDegrees=45
  twistMotion=Free
 }
 {
  bone1Name=backpack_spine_03
  bone2Name=backpack_spine_02
  swing1LimitDegrees=45
  swing1Motion=Free
  swing2LimitDegrees=45
  swing2Motion=Free
  twistLimitDegrees=45
  twistMotion=Free
 }
 {
  bone1Name=backpack_spine_04
  bone2Name=backpack_spine_03
  swing1LimitDegrees=45
  swing1Motion=Free
  swing2LimitDegrees=45
  swing2Motion=Free
  twistLimitDegrees=45
  twistMotion=Free
 }
 {
  bone1Name=backpack_spine_05
  bone2Name=backpack_spine_04
  swing1LimitDegrees=45
  swing1Motion=Free
  swing2LimitDegrees=45
  swing2Motion=Free
  twistLimitDegrees=45
  twistMotion=Free
 }
 {bone1Name=dyn_pack_04, bone2Name=dyn_pack_03, swing1LimitDegrees=10, swing2LimitDegrees=10, twistLimitDegrees=5}
 {bone1Name=dyn_pack_05, bone2Name=dyn_pack_04, swing1LimitDegrees=5, swing2LimitDegrees=10, twistLimitDegrees=3}
 {bone1Name=dyn_pack_01, bone2Name=pack, swing1LimitDegrees=5, swing2LimitDegrees=20, twistLimitDegrees=1}
 {bone1Name=dyn_pack_02, bone2Name=dyn_pack_01, swing1LimitDegrees=7, swing2LimitDegrees=25, twistLimitDegrees=1}
 {bone1Name=dyn_pack_03, bone2Name=dyn_pack_02, swing1LimitDegrees=10, swing2LimitDegrees=20, twistLimitDegrees=5}
]
